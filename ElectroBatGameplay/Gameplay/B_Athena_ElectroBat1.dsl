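; Electro Bat weapon Blueprint that toggles the bat's special movement and Niagara FX on equip and handles its charge and swing cues.
; ElectroBatWeaponComponent drives charging; a tag query or vehicle state change stops it, as do unequip and EndPlay.
; Charge events update looping GameplayCues indexed by the CurrentChargeStage attribute and store the matching burst swing cue.
; Cue application waits a tick to escape prediction-key scope, which would suppress cues.
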(event FortWeapon|EventOnEquippedAndAttached
  ;[0.05381 0.15 0.1098 0.5] On Equip
  (|CallParentFunction)
  (bind _electrobatweaponcomponent (Variables|Default|GetElectroBatWeaponComponent)
    (:then
      (Class|ElectroBatWeaponComponent|SetElectroBatMovementEnabled _electrobatweaponcomponent true)
      (Components|Activation|Activate (Variables|Default|GetIdleFX%20Niagara) false)
      (Utilities|FlowControl|Delay 0.52)
      (Components|Activation|Activate (Variables|Default|GetNS_ElectroBat_Equip) true))
    (:else)))

(event FortWeapon|EventOnLocalUnEquip
  ;[0.15 0.07841 0.01641 0.5] Cleanup: On Un-Equip + End Play
  (|CallParentFunction)
  (bind _electrobatweaponcomponent (Variables|Default|GetElectroBatWeaponComponent)
    (:then
      (Class|ElectroBatWeaponComponent|StopCharging _electrobatweaponcomponent)
      (Class|ElectroBatWeaponComponent|SetElectroBatMovementEnabled _electrobatweaponcomponent false)
      (Components|Activation|Deactivate (Variables|Default|GetIdleFX%20Niagara))
      (Components|Activation|Deactivate (Variables|Default|GetNS_ElectroBat_Equip)))
    (:else)))

(event EventEndPlay
  (params (EndPlayReason :type EEndPlayReason%20Enum :default "Destroyed"))
  ;[0.15 0.07841 0.01641 0.5] Cleanup: On Un-Equip + End Play
  (|CallParentFunction EndPlayReason)
  (bind _electrobatweaponcomponent (Variables|Default|GetElectroBatWeaponComponent)
    (:then
      (Class|ElectroBatWeaponComponent|StopCharging _electrobatweaponcomponent)
      (Class|ElectroBatWeaponComponent|SetElectroBatMovementEnabled _electrobatweaponcomponent false)
      (Components|Activation|Deactivate (Variables|Default|GetIdleFX%20Niagara))
      (Components|Activation|Deactivate (Variables|Default|GetNS_ElectroBat_Equip)))
    (:else)))

(event OnChargeStart(ElectroBatWeaponComponent)
  ;[6.1e-05 0.3871 0.4031 0.5] Charging
  (CallFunction|StartWaitforChargeStop)
  (Networking|SwitchHasAuthority
    (:Authority
      (Variables|Charging|GameplayCues|SetCurrentChargingGameplayCueTag)
      (CallFunction|PlayChargingFX))))

(event Custom|Start%20Charging
  ;[6.1e-05 0.3871 0.4031 0.5] Charging
  (bind _electrobatweaponcomponent (Variables|Default|GetElectroBatWeaponComponent)
    (:then
      (Class|ElectroBatWeaponComponent|StartCharging _electrobatweaponcomponent))
    (:else)))

(event OnChargeStop(ElectroBatWeaponComponent)
  ;[6.1e-05 0.3871 0.4031 0.5] Charging
  (CallFunction|StopWaitforChargeStop)
  (Networking|SwitchHasAuthority
    (:Authority
      (CallFunction|RemoveChargingFX))))

(event Custom|Stop%20Charging
  ;[6.1e-05 0.3871 0.4031 0.5] Charging
  (bind _electrobatweaponcomponent (Variables|Default|GetElectroBatWeaponComponent)
    (:then
      (Class|ElectroBatWeaponComponent|StopCharging _electrobatweaponcomponent))
    (:else)))

(event OnChargeIncreased(ElectroBatWeaponComponent)
  ;[6.1e-05 0.3871 0.4031 0.5] Charging
  (CallFunction|PlayChargingFX))

(event Custom|Start%20Wait%20For%20Charge%20Stop
  (bind _returnvalue (Game|GetInstigator))
  ;[0.02907 0.15 0.09005 0.5] Wait for stop charge
  (Utilities|IsValid _returnvalue
    (:Is%20Valid
      (bind _asynctaskproxy (Ability|Tasks|WaitGameplayTagQueryOnActor _returnvalue (Variables|Charging|GetEndChargingTagQuery) "WhenTrue" true)
        (:then
          (Variables|Charging|SetWaitForEndChargeTagQueryAction _asynctaskproxy)
          (bind _asfort_player_pawn (Utilities|Casting|CastToFortPlayerPawn _returnvalue))
          (Vehicle|BindEventtoOnVehicleStateChangedEvent _asfort_player_pawn (delegate Custom|On%20Vehicle%20State%20Changed)))
        (:Triggered
          (CallFunction|StopCharging))))))

(event Custom|On%20Vehicle%20State%20Changed
  (params (FortPlayerPawn :type Fort%20Player%20Pawn%20Object%20Reference) (NewVehicle :type Actor%20Object%20Reference) (OldVehicle :type Actor%20Object%20Reference))
  ;[0.02907 0.15 0.09005 0.5] Wait for stop charge
  (CallFunction|StopCharging))

(event Custom|Stop%20Wait%20For%20Charge%20Stop
  ;[0.02907 0.15 0.09005 0.5] Wait for stop charge
  (bind _waitforendchargetagqueryaction (Variables|Charging|GetWaitForEndChargeTagQueryAction)
    (:then
      (Ability|Async|EndAction _waitforendchargetagqueryaction))
    (:else
      _))
  (bind _asfort_player_pawn (Utilities|Casting|CastToFortPlayerPawn (Game|GetInstigator)))
  (Vehicle|UnbindEventfromOnVehicleStateChangedEvent _asfort_player_pawn (EventDispatchers|CreateEvent :!event Custom|On%20Vehicle%20State%20Changed)))

(event Custom|Play%20Charging%20FX
  (bind _returnvalue (Game|GetInstigator))
  (bind _currentchargestate (CallFunction|GetCurrentChargeStage))
  (bind _chargegameplaycues (Variables|Charging|GameplayCues|GetChargeGameplayCues))
  (bind _output (Utilities|Array|Get(acopy) _chargegameplaycues _currentchargestate))
  ;[0.3531 0.01828 0.3324 0.5] FX
  ;;[0.1156 0.04428 0.15 0.5] Play Charging FX - Apply the GC for the current charge count to the player and update current GC tags
  (Utilities|FlowControl|DelayUntilNextTick) ; Ensures GCs are not applied within a scope with a prediction key. \nOtherwise GA system will ignore GCs
  (if (and (Utilities|Array|IsValidIndex (Variables|Charging|GameplayCues|GetSwingGameplayCues) _currentchargestate) (Utilities|Array|IsValidIndex _chargegameplaycues _currentchargestate))
    (if (and (Utilities|IsValid _returnvalue :!class /Script/Engine.KismetSystemLibrary) (!= _output (Variables|Charging|GameplayCues|GetCurrentChargingGameplayCueTag)))
      (bind _output_get (Variables|Charging|GameplayCues|SetCurrentChargingGameplayCueTag _output))
      (GameplayCue|AddGameplayCueOnActor(Looping) _returnvalue _output_get (Variables|Charging|GameplayCues|GetChargeGameplayCueParams))
      (Variables|Charging|GameplayCues|SetCurrentSwingGameplayCueTag (Utilities|Array|Get(aref) (Variables|Charging|GameplayCues|GetSwingGameplayCues) (CallFunction|GetCurrentChargeStage))))
    (else
      ;;;[0.15 0.000781 0.000781 0.5] GC array validation
      (Fortnite|FortShippingLog self "B_Athena_ElectroBat: No GC tag specified for current charge level" true))))

(event Custom|Remove%20Charging%20FX
  (bind _returnvalue (Game|GetInstigator))
  ;[0.3531 0.01828 0.3324 0.5] FX
  ;;[0.15 0.02573 0.06695 0.5] Remove Charging FX - Ensure all charging GCs are removed from the player
  (Utilities|IsValid _returnvalue
    (:Is%20Valid
      (Variables|Charging|GameplayCues|SetCurrentChargingGameplayCueTag)
      (for _array_element (Variables|Charging|GameplayCues|GetChargeGameplayCues)
        (GameplayCue|RemoveGameplayCueOnActor(Looping) _returnvalue _array_element (Variables|Charging|GameplayCues|GetChargeGameplayCueParams))))))

(event Custom|Play%20Swing%20FX
  (bind _returnvalue (Game|GetInstigator))
  ;[0.3531 0.01828 0.3324 0.5] FX
  ;;[0.03196 0.03196 0.15 0.5] Play Swing FX - Execute the swing GC for the current charge state
  (Utilities|IsValid _returnvalue
    (:Is%20Valid
      (GameplayCue|ExecuteGameplayCueOnActor(Burst) _returnvalue (Variables|Charging|GameplayCues|GetCurrentSwingGameplayCueTag) (Variables|Charging|GameplayCues|GetChargeGameplayCueParams))
      (Variables|Charging|GameplayCues|SetCurrentSwingGameplayCueTag))))

(fn GetCurrentChargeStage
  (bind _returnvalue (Game|GetInstigator))
  (bind (_bsuccessfullyfoundattribute _returnvalue_2) (Ability|Attribute|GetFloatAttribute _returnvalue "(AttributeName=\"CurrentChargeStage\",Attribute=/Script/ElectroBatGameplayRuntime.ElectroBatAttributeSet:CurrentChargeStage,AttributeOwner=\"/Script/CoreUObject.Class'/Script/ElectroBatGameplayRuntime.ElectroBatAttributeSet'\")"))
  (Utilities|IsValid _returnvalue
    (:Is%20Valid
      (return (Math|Float|Truncate _returnvalue_2)))))

(fn IsCharging
  (bind _electrobatweaponcomponent (Variables|Default|GetElectroBatWeaponComponent)
    (:then
      (return (Class|ElectroBatWeaponComponent|IsCharging _electrobatweaponcomponent)))
    (:else)))

(fn ConstructionScript
  (|CallParentFunction))
