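; ElectroBat knockback gameplay ability that launches its avatar with a strength based on the KnockbackCharge attribute.
; On activation it checks team affiliation with the granting effect's causer unless teammate knockback is allowed, then stores the avatar's evaluated charge and runs the parent logic.
; The GetInitialKnockMag override feeds the launcher's forward vector to the parent and scales the result between a low-knockback percentage and full strength by that charge.
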(event Ability|EventActivateAbility
  (bind _returnvalue (Ability|GetAvatarActorfromActorInfo))
  (bind _returnvalue_2 (Ability|GetAbilitySystemComponent _returnvalue))
  (bind (_bsuccess _returnvalue_3) (Ability|Attribute|EvaluateAttributeValuewithTags _returnvalue_2 "(AttributeName=\"KnockbackCharge\",Attribute=/Script/ElectroBatGameplayRuntime.ElectroBatAttributeSet:KnockbackCharge,AttributeOwner=\"/Script/CoreUObject.Class'/Script/ElectroBatGameplayRuntime.ElectroBatAttributeSet'\")" (Variables|Default|GetSource%20Tags) (Variables|Default|GetTarget%20Tags)))
  (if (not (ScalableFloat|GetValueasBool (Variables|Default|GetRow%20Knockback%20Allow%20Teammate%20Knockback)))
    (switch Utilities|FlowControl|Switch|SwitchonEFortTeamAffiliation (Team|GetTeamAffiliation (Ability|GetAvatarActorfromActorInfo) (Ability|EffectContext|GetEffectCauser (Ability|GetGrantedbyEffectContext)))
      (:Neutral
        _)
      (:Hostile
        _)))
  (Utilities|IsValid _returnvalue
    (:Is%20Valid
      (Utilities|IsValid _returnvalue_2
        (:Is%20Valid
          (Variables|Default|SetKnockback%20Magnitude _returnvalue_3)
          (|CallParentFunction))))))

(fn GetInitialKnockMag
  (params (Direction :type Vector :default "0, 0, 0") (OverrideXYMag :type Boolean) (XYMAgOverride :type Float%20%28double-precision%29))
  (return (* (|CallParentFunction (Transformation|GetActorForwardVector (Variables|KnockbackParent|GetLauncher))) (Math|Float|Lerp (ScalableFloat|GetValueatLevel (Variables|Default|GetRow%20Low%20Knockback%20Percentage)) 1.0 (Variables|Default|GetKnockback%20Magnitude)))))
