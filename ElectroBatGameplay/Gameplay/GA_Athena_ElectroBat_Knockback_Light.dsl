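; Light knockback gameplay ability for the ElectroBat, built on a knockback parent class.
; It overrides GetInitialKnockMag to call the parent version with the launcher's actor forward vector as the direction, so the knock is based on the launcher's facing.
; The incoming Direction and XY magnitude override arguments are not forwarded to the parent.
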(fn GetInitialKnockMag
  (params (Direction :type Vector :default "0, 0, 0") (OverrideXYMag :type Boolean) (XYMAgOverride :type Float%20%28double-precision%29))
  (return (|CallParentFunction (Transformation|GetActorForwardVector (Variables|KnockbackParent|GetLauncher)))))
