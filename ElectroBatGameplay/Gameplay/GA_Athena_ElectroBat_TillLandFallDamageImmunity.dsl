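; ElectroBat gameplay ability that, by its name, gives fall damage immunity until landing, likely to knockback targets.
; On activation it compares the avatar pawn's team with the source weapon's.
; For friendly pawns, it ends the ability if a scalable-float setting disallows teammate knockback.
; Otherwise it reads the grant-immunity-to-target setting for non-friendly pawns, with no visible branch actions, and relies on the parent activation for the effect itself.
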(event Ability|EventActivateAbility
  (if (Utilities|Enum|Equal(Enum) (Team|GetTeamAffiliation (CallFunction|GetAvatarasFortPlayerPawn) (Utilities|Casting|CastToFortWeapon (Ability|GetCurrentSourceObject))) "Friendly")
    (if (not (ScalableFloat|GetValueasBool (Variables|Default|GetRow%20Knockback%20Allow%20Teammate%20Knockback)))
      (Ability|EndAbility))
    (else
      (if (ScalableFloat|GetValueasBool (Variables|Default|GetRow%20Knockback%20Grant%20FDI%20to%20Target))
        _
        (else
          _))))
  (|CallParentFunction))
